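# Level Sequence for animating the FacetBlom_M custom skin in the BR main-player menu, apparently an animation authoring scene.
# A spawned FortPlayerMannequin holds at an identity transform and carries an MR_Biped_v02 Control Rig track, while its skeletal animation track is empty.
# Its child Body component adds a Fortnite_ExtraJoint_CtrlRig track, so posing comes from the two rigs rather than animation clips.
# The sequence has no camera or other top-level tracks.
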LevelSequence /BRCosmetics/Animation/Game/MainPlayer/Menu/BR/CustomSkins/_s40/FacetBlom_M/FacetBlom_M_SL
bindings [
 FortPlayerMannequin {
  id=#1
  possessed_class=FortPlayerMannequin
  tracks [
   {
    type=MovieSceneSpawnTrack
    sections [{type=MovieSceneSpawnSection, channels [None {type=MovieSceneScriptingBoolChannel, default=true, const [true,3]}]}]
   }
   {
    type=MovieScene3DTransformTrack
    sections [{
      type=MovieScene3DTransformSection
      channels [
       Location.X {type=MovieSceneScriptingDoubleChannel, default=0, const [0,26]}
       Location.Y {type=MovieSceneScriptingDoubleChannel, default=0, const [0,26]}
       Location.Z {type=MovieSceneScriptingDoubleChannel, default=0, const [0,26]}
       Rotation.X {type=MovieSceneScriptingDoubleChannel, default=0, const [0,26]}
       Rotation.Y {type=MovieSceneScriptingDoubleChannel, default=0, const [0,26]}
       Rotation.Z {type=MovieSceneScriptingDoubleChannel, default=0, const [0,26]}
       Scale.X {type=MovieSceneScriptingDoubleChannel, default=1, const [1,26]}
       Scale.Y {type=MovieSceneScriptingDoubleChannel, default=1, const [1,26]}
       Scale.Z {type=MovieSceneScriptingDoubleChannel, default=1, const [1,26]}
      ]
     }]
   }
   {type=MovieSceneSkeletalAnimationTrack, sections []}
   {type=MovieSceneControlRigParameterTrack, display_name=MR_Biped_v02, sections [{type=MovieSceneControlRigParameterSection}]}
  ]
 }
 Body {
  id=#2
  parent_id=#1
  possessed_class=SkeletalMeshComponentBudgeted
  tracks [
   {type=MovieSceneControlRigParameterTrack, display_name=Fortnite_ExtraJoint_CtrlRig, sections [{type=MovieSceneControlRigParameterSection}]}
   {
    type=MovieScene3DTransformTrack
    sections [{
      type=MovieScene3DTransformSection
      channels [
       Rotation.X {type=MovieSceneScriptingDoubleChannel, default=0}
       Rotation.Y {type=MovieSceneScriptingDoubleChannel, default=0}
       Rotation.Z {type=MovieSceneScriptingDoubleChannel, default=0}
      ]
     }]
   }
  ]
 }
]
display_rate {denominator=1, numerator=30}
tick_resolution {denominator=1, numerator=24000}
tracks []
